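# Sound cue for the DgMl pet's DBNO (down but not out) start sound.
# A quality level node feeds its first two slots from one shared random pick between two equally weighted waves, and its last two from a single "_opt" wave.
# The cue uses the pets sound class and close-range pet attenuation, with boosted volume.
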SoundCue /BRCosmetics/Sounds/Pets/DgMl/PT_DgMl_DbnoStart_Cue
root=QualityLevel_0
nodes [
 {id=QualityLevel_0, children [Random_0,Random_0,WavePlayer_0,WavePlayer_0]}
 {id=Random_0, children [WavePlayer_1,WavePlayer_2], properties {weights [1,1]}}
 {id=WavePlayer_0, properties {soundWaveAssetPtr=&/PT_DgMl_DbnoStart_01_opt}}
 {id=WavePlayer_1, properties {soundWaveAssetPtr=&/PT_DgMl_DbnoStart_01}}
 {id=WavePlayer_2, properties {soundWaveAssetPtr=&/PT_DgMl_DbnoStart_02}}
]
properties {
 attenuationSettings=&Fort_Pets_Close_Athena_Att
 maxDistance=1800.0001
 soundClassObject=&Fort_Creatures_Pets
 volumeMultiplier=1.68
}
