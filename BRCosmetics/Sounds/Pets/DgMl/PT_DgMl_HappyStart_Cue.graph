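# Sound cue for the DgMl pet's HappyStart sound, a short-range creature pet sound with close pet attenuation and reduced volume.
# A quality-level node sends the first two levels to an equally weighted random pick among four Treasure_17 wave variants.
# The last two levels share a single "opt" wave, skipping the randomization.
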SoundCue /BRCosmetics/Sounds/Pets/DgMl/PT_DgMl_HappyStart_Cue
root=QualityLevel_0
nodes [
 {id=QualityLevel_0, children [Random_0,Random_0,WavePlayer_4,WavePlayer_4]}
 {id=Random_0, children [WavePlayer_0,WavePlayer_1,WavePlayer_2,WavePlayer_3], properties {weights [1,1,1,1]}}
 {id=WavePlayer_0, properties {soundWaveAssetPtr=&/PT_DgMl_Treasure_17}}
 {id=WavePlayer_1, properties {soundWaveAssetPtr=&/PT_DgMl_Treasure_17b}}
 {id=WavePlayer_2, properties {soundWaveAssetPtr=&/PT_DgMl_Treasure_17c}}
 {id=WavePlayer_3, properties {soundWaveAssetPtr=&/PT_DgMl_Treasure_17d}}
 {id=WavePlayer_4, properties {soundWaveAssetPtr=&/PT_DgMl_Treasure_17_opt}}
]
properties {
 attenuationSettings=&Fort_Pets_Close_Athena_Att
 maxDistance=1800.0001
 soundClassObject=&Fort_Creatures_Pets
 volumeMultiplier=0.448
}
