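# Sound cue for the skydiving sound of the DgMl cosmetic pet, routed to the creature pets sound class with close-range pet attenuation.
# A quality switch sends the first two levels to a weighted random pick between two waves, favoring variant 01, with repeats allowed.
# The other two levels play a single optimized version of variant 02.
# A concurrency override permits one instance per owner and prevents new plays while it is active.
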SoundCue /BRCosmetics/Sounds/Pets/DgMl/PT_DgMl_Skydiving_Cue
root=QualityLevel_0
nodes [
 {id=QualityLevel_0, children [Random_0,Random_0,WavePlayer_3,WavePlayer_3]}
 {id=Random_0, children [WavePlayer_0,WavePlayer_1], properties {bRandomizeWithoutReplacement=false, weights [1,3]}}
 {id=WavePlayer_0, properties {soundWaveAssetPtr=&/PT_DgMl_Skydiving_02}}
 {id=WavePlayer_1, properties {soundWaveAssetPtr=&/PT_DgMl_Skydiving_01}}
 {id=WavePlayer_3, properties {soundWaveAssetPtr=&/PT_DgMl_Skydiving_02_opt}}
]
properties {
 attenuationSettings=&Fort_Pets_Close_Athena_Att
 bOverrideConcurrency=true
 maxDistance=1800.0001
 soundClassObject=&Fort_Creatures_Pets
 volumeMultiplier=0.672
 concurrencyOverrides {bLimitToOwner=true, maxCount=1, resolutionRule=PreventNew}
}
